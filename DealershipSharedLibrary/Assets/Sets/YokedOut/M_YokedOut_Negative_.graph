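# Translucent unlit material for Niagara mesh particles in the YokedOut set, giving a negative-image look by inverting the scene color behind the mesh.
# A saturated Fresnel term plus a lightning texture blends that inversion toward dark purple, with a constant replacing scene color on lower feature levels.
# Dynamic parameters drive the texture's tiling, offset and channel selection.
# A panned Perlin noise branch builds a scaled world position offset signal.
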Material /DealershipSharedLibrary/Assets/Sets/YokedOut/M_YokedOut_Negative_
graph {
 properties {
  bUsedWithNiagaraMeshParticles=true
  bUsedWithStaticMesh=false
  blendMode=BLEND_Translucent
  previewMesh=/Engine/Transient.MaterialEditorMeshComponent_44
  shadingModel=MSM_Unlit
  usedShadingModels=Unlit
 }
 nodes [
  n1 = SceneColor!(UVs)
  n2 = Constant()
  n3 = ScalarParameter() {defaultValue=10, parameterName=ExponentIn}
  n4 = Saturate(None=LinearInterpolate(
    A=ScalarParameter() {parameterName=A}
    B=ScalarParameter() {defaultValue=20, parameterName=B}
    Alpha=Fresnel(ExponentIn=@n3) {baseReflectFraction=0, exponent=10}
   ) {constA=-0.1051, constB=20})
  n5 = DynamicParameter() {paramNames [Tex1 Tiling U, Tex1 Tiling V, Tex1 Offset U, Tex1 Offset V]}
  n6 = Multiply(A=TextureCoordinate(), B=AppendVector(A=@n5.Tex1 Tiling U, B=@n5.Tex1 Tiling U))
  n7 = AppendVector(A=@n5.Tex1 Offset U, B=@n5.Tex1 Offset U)
  n8 = TextureSampleParameter2D(UVs=Add(B=Add(A=@n6, B=@n7))) {group=Texture 1, parameterName=Texture 1, texture=/DealershipSharedLibrary/Assets/Sets/Lightning/T_VCVFX_Lightning}
  n9 = DynamicParameter() {parameterIndex=1, paramNames [Channel Select, Param2, Param3, Param4]}
  n10 = LinearInterpolate(A=@n8.R, B=@n8.R, Alpha=@n9.Channel Select)
  n11 = Saturate(None=Subtract(A=@n9.Channel Select))
  n12 = Power(
   Base=Multiply(
    A=LinearInterpolate(A=@n10, B=@n8.B, Alpha=@n11)
    B=ScalarParameter() {defaultValue=1, group=Texture 1, parameterName=Texture 1 - Multiplier}
   )
   Exp=ScalarParameter() {defaultValue=1, group=Texture 1, parameterName=Texture 1 - Power}
  )
  LinearInterpolate(
   A=FeatureLevelSwitch(Default=OneMinus(None=@n1.RGB), ES2=@n2, ES3_1=@n2, SM4_REMOVED=@n2)
   B=Constant3Vector() {constant=<rgba 0.043 0 0.1912 1>}
   Alpha=Saturate(None=Add(A=@n4, B=@n12))
  )
  n13 = TextureCoordinate()
  n14 = VectorParameter() {
   group=WPO
   parameterName="WPO -  Noise UV Controls"
   channelNames {a=Panner V, b=Panner U, g=Tiling V, r=Tiling U}
   defaultValue=<rgba 0.5821 0.4286 0 0.128>
  }
  n15 = AppendVector(A=@n14.Tiling U, B=@n14.Tiling U)
  n16 = TextureSampleParameter2D(UVs=Panner(Coordinate=Multiply(A=@n13, B=@n15) {constB=2}, Speed=AppendVector(A=@n14.Panner U, B=@n14.Panner U)) {speedX=0.2, speedY=0.15}) {group=WPO, parameterName=WPO Noise, texture=/Game/Effects/Fort_Effects/Textures/Tile/T_PerlinRGB_01_M}
  Multiply(A=VertexNormalWS(), B=@n16.R)
  NamedRerouteDeclaration(None=Multiply(A=@n16.RGB, B=ScalarParameter() {defaultValue=5, parameterName=WPO - Amount})) {name=WPO_Out, nodeColor=<rgba 1 0.9098 0.2641 1>}
  NamedRerouteUsage()
 ]
}
